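# Physics asset for the left calf of the male medium InspireSpell player body cosmetic, set up for dangling bootlace physics.
# Calf and boot capsules share the calf body, with small capsules on inner and outer chains of three lace bones and two loop bones each.
# Each chain hangs from the calf through angular limits, generally tighter for loops.
# The calf-to-thigh constraint rotates freely, though the thigh has no body here.
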PhysicsAsset /BRCosmetics/Characters/Player/Male/Medium/Bodies/M_MED_InspireSpell/Meshes/M_MED_InspireSpell_CalfL_Physics
bodies {
 calf_l [
  {shapeType=Capsule, length=24, radius=7.8304, shapeName=calf_l_capsule, center=<-7.9438 -5.0162 0>, rotation=<rot -90 0 0>}
  {
   shapeType=Capsule
   length=9.0091
   radius=5.5035
   shapeName=boot_capsule
   center=<-32.1286 -1.5965 0.8449>
   rotation=<rot -87.1335 -179.9998 177.1635>
  }
 ]
 dyn_lace_inner_01_l [{
   shapeType=Capsule
   length=0.5084
   radius=0.5326
   shapeName=dyn_lace_inner_01_l_capsule
   center=<0.5743 0 0>
   rotation=<rot -90 -165.9884 166.0149>
  }]
 dyn_lace_inner_02_l [{shapeType=Capsule, length=0.8997, radius=0.5194, shapeName=dyn_lace_inner_02_l_capsule, center=<0.8563 0 0>, rotation=<rot -90 180 180>}]
 dyn_lace_inner_03_l [{shapeType=Capsule, length=1.25, radius=0.6295, shapeName=dyn_lace_inner_03_l_capsule, center=<1.0255 0 0>, rotation=<rot -90 180 180>}]
 dyn_lace_outer_01_l [{shapeType=Capsule, length=0.5699, radius=0.5, shapeName=dyn_lace_outer_01_l_capsule, center=<0.4672 0 0>, rotation=<rot -90 180 180>}]
 dyn_lace_outer_02_l [{shapeType=Capsule, length=0.6339, radius=0.5094, shapeName=dyn_lace_outer_02_l_capsule, center=<0.8376 0 0>, rotation=<rot -90 180 180>}]
 dyn_lace_outer_03_l [{shapeType=Capsule, length=1.2621, radius=0.6295, shapeName=dyn_lace_outer_03_l_capsule, center=<1.1833 0 0>, rotation=<rot -90 180 180>}]
 dyn_loop_inner_01_l [{shapeType=Capsule, length=0.9825, radius=0.5, shapeName=dyn_loop_inner_01_l_capsule, center=<0.8505 0 0>, rotation=<rot -90 180 180>}]
 dyn_loop_inner_02_l [{shapeType=Capsule, length=1.5, radius=0.5, shapeName=dyn_loop_inner_02_l_capsule, center=<1.2031 0 0>, rotation=<rot -90 0 0>}]
 dyn_loop_outer_01_l [{shapeType=Capsule, length=0.6639, radius=0.5443, shapeName=dyn_loop_outer_01_l_capsule, center=<0.8661 0 0>, rotation=<rot -90 180 180>}]
 dyn_loop_outer_02_l [{shapeType=Capsule, length=1.25, radius=0.6227, shapeName=dyn_loop_outer_02_l_capsule, center=<0.8812 0 0>, rotation=<rot -90 180 180>}]
}
constraints [
 {
  bone1Name=calf_l
  bone2Name=thigh_l
  swing1LimitDegrees=45
  swing1Motion=Free
  swing2LimitDegrees=45
  swing2Motion=Free
  twistLimitDegrees=45
  twistMotion=Free
 }
 {bone1Name=dyn_lace_inner_01_l, bone2Name=calf_l, swing1LimitDegrees=30, swing2LimitDegrees=45, twistLimitDegrees=4}
 {bone1Name=dyn_lace_outer_01_l, bone2Name=calf_l, swing1LimitDegrees=30, swing2LimitDegrees=45, twistLimitDegrees=4}
 {bone1Name=dyn_loop_inner_01_l, bone2Name=calf_l, swing1LimitDegrees=20, swing2LimitDegrees=10, twistLimitDegrees=1}
 {bone1Name=dyn_loop_outer_01_l, bone2Name=calf_l, swing1LimitDegrees=20, swing2LimitDegrees=10, twistLimitDegrees=1}
 {bone1Name=dyn_lace_inner_02_l, bone2Name=dyn_lace_inner_01_l, swing1LimitDegrees=30, swing2LimitDegrees=30, twistLimitDegrees=4}
 {bone1Name=dyn_lace_inner_03_l, bone2Name=dyn_lace_inner_02_l, swing1LimitDegrees=35, swing2LimitDegrees=35, twistLimitDegrees=4}
 {bone1Name=dyn_lace_outer_02_l, bone2Name=dyn_lace_outer_01_l, swing1LimitDegrees=30, swing2LimitDegrees=30, twistLimitDegrees=4}
 {bone1Name=dyn_lace_outer_03_l, bone2Name=dyn_lace_outer_02_l, swing1LimitDegrees=40, swing2LimitDegrees=40, twistLimitDegrees=45}
 {bone1Name=dyn_loop_inner_02_l, bone2Name=dyn_loop_inner_01_l, swing1LimitDegrees=30, swing2LimitDegrees=15, twistLimitDegrees=1}
 {bone1Name=dyn_loop_outer_02_l, bone2Name=dyn_loop_outer_01_l, swing1LimitDegrees=30, swing2LimitDegrees=15, twistLimitDegrees=1}
]
