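# Control Rig for the M_MED_InspireSpell male body that procedurally offsets the dynamic shorts bones as the knees bend.
# Constant offsets first shift each leg's root shorts bones and rear tip bones.
# Each calf's local Z rotation is then remapped to an alpha, smoothed by AlphaInterp, and used to weight hand-tuned, asymmetric corrective offsets.
# The isCrouching and isJumping variables are unused; an axis-draw node is unconnected.
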ControlRigBlueprint /BRCosmetics/Characters/Player/Male/Medium/Bodies/M_MED_InspireSpell/Meshes/M_MED_InspireSpell_CtrlRig
graphs [
 ContainedGraph {nodes [
   n1 = RigVMFunctionEntryNode()
   SequenceAggregate = &RigVMFunction_Sequence(ExecuteContext=@n1.ExecuteContext)
   SequenceAggregate_1 = &RigVMFunction_Sequence(ExecuteContext=@SequenceAggregate.B)
   SequenceAggregate_2 = &RigVMFunction_Sequence(ExecuteContext=@SequenceAggregate_1.B)
   SequenceAggregate_3 = &RigVMFunction_Sequence(ExecuteContext=@SequenceAggregate_2.B)
   RigVMFunctionReturnNode(A=@SequenceAggregate.A, B=@SequenceAggregate_1.A, C=@SequenceAggregate_2.A, D=@SequenceAggregate_3.A, E=@SequenceAggregate_3.B)
  ]}
 RigVMFunctionLibrary {links [], nodes []}
 RigVMModel {nodes [
   RigUnit_BeginExecution = &RigUnit_BeginExecution()
   SequenceAggregate = RigVMAggregateNode(ExecuteContext=@RigUnit_BeginExecution.ExecutePin)
   n1 = RigVMDispatchNode(ExecuteContext=@SequenceAggregate.B, Array=RigVMVariableNode() {variable=BonesToMove_R}.Value.Keys) {template="DISPATCH_RigVMDispatch_ArrayIterator(in Array,out Element,out Index,out Count,out Ratio)"}
   > &RigUnit_ModifyTransforms(
    ItemToModify=((Item=(Type=Bone),Transform=(Rotation=(X=0,Y=0,Z=0,W=1),Translation=(X=3,Y=0,Z=0),Scale3D=(X=1,Y=1,Z=1))))
    Weight=1
    ItemToModify.0.Item=@n1.Element
   )
   &RigUnit_ModifyTransforms(
    ItemToModify=((Item=(Type=Bone,Name="dyn_shorts_back_03_l"),Transform=(Rotation=(X=0,Y=0,Z=0,W=1),Translation=(X=-4,Y=0,Z=0),Scale3D=(X=1,Y=1,Z=1))))
    Weight=1
    ExecutePin=@SequenceAggregate.C
   )
   > &RigUnit_ModifyTransforms(
    ItemToModify=((Item=(Type=Bone,Name="dyn_shorts_back_03_r"),Transform=(Rotation=(X=0,Y=0,Z=0,W=1),Translation=(X=4,Y=0,Z=0),Scale3D=(X=1,Y=1,Z=1))))
    Weight=1
   )
   SequenceAggregate_1 = &RigVMFunction_Sequence(ExecuteContext=@SequenceAggregate.D)
   Remap = &RigVMFunction_MathDoubleRemap(
    SourceMaximum=115
    SourceMinimum=22
    TargetMaximum=1
    TargetMinimum=0
    Value=&RigVMFunction_MathQuaternionToEuler(RotationOrder=XYZ, Value=&RigUnit_GetTransform(Item=(Type=Bone,Name="calf_l"), Space=LocalSpace).Transform.Rotation).Result.Z
   )
   RigVMVariableNode(Value=@Remap.Result, ExecuteContext=@SequenceAggregate_1.A) {variable=calf_l_Alpha}
   n2 = &RigVMFunction_AlphaInterp(InterpSpeedDecreasing=10, InterpSpeedIncreasing=10, Scale=1, bClampResult=False, Value=RigVMVariableNode() {variable=calf_l_Alpha}.Value)
   > &RigUnit_ModifyTransforms(
    ItemToModify=((Item=(Type=Bone,Name="dyn_shorts_back_03_l"),Transform=(Rotation=(X=0,Y=0,Z=-0.0872,W=0.9962),Translation=(X=-4,Y=0,Z=0),Scale3D=(X=1,Y=1,Z=1))))
    Weight=@n2.Result
   )
   > &RigUnit_ModifyTransforms(
    ItemToModify=((Item=(Type=Bone,Name="dyn_shorts_back_02_l"),Transform=(Rotation=(X=0,Y=0,Z=0,W=1),Translation=(X=-6,Y=0,Z=0),Scale3D=(X=1,Y=1,Z=1))))
    Weight=@n2.Result
   )
   > &RigUnit_ModifyTransforms(
    ItemToModify=((Item=(Type=Bone,Name="dyn_shorts_inner_03_l"),Transform=(Rotation=(X=0,Y=0,Z=0,W=1),Translation=(X=-16,Y=0,Z=0),Scale3D=(X=1,Y=1,Z=1))))
    Weight=@n2.Result
   )
   > &RigUnit_ModifyTransforms(
    ItemToModify=((Item=(Type=Bone,Name="dyn_shorts_outer_03_l"),Transform=(Rotation=(X=0,Y=0,Z=0,W=1),Translation=(X=-10,Y=0,Z=0),Scale3D=(X=1,Y=1,Z=1))))
    Weight=@n2.Result
   )
   n3 = &RigUnit_GetTransform(Item=(Type=Bone,Name="calf_r"), Space=LocalSpace)
   RigVMVariableNode(
    Value=&RigVMFunction_MathDoubleRemap(
     SourceMaximum=119
     SourceMinimum=19
     TargetMaximum=1
     TargetMinimum=0
     Value=&RigVMFunction_MathQuaternionToEuler(RotationOrder=XYZ, Value=@n3.Transform.Rotation).Result.Z
    ).Result
    ExecuteContext=@SequenceAggregate_1.B
   ) {variable=calf_r_Alpha}
   AlphaInterp = &RigVMFunction_AlphaInterp(InterpSpeedDecreasing=10, InterpSpeedIncreasing=10, Scale=1, bClampResult=False, Value=RigVMVariableNode() {variable=calf_r_Alpha}.Value)
   > &RigUnit_ModifyTransforms(
    ItemToModify=((Item=(Type=Bone,Name="dyn_shorts_back_03_r"),Transform=(Rotation=(X=0,Y=0,Z=0,W=1),Translation=(X=4,Y=0,Z=0),Scale3D=(X=1,Y=1,Z=1))))
    Weight=@AlphaInterp.Result
   )
   > &RigUnit_ModifyTransforms(
    ItemToModify=((Item=(Type=Bone,Name="dyn_shorts_back_02_r"),Transform=(Rotation=(X=0,Y=0,Z=0,W=1),Translation=(X=8,Y=-2,Z=0),Scale3D=(X=1,Y=1,Z=1))))
    Weight=@AlphaInterp.Result
   )
   > &RigUnit_ModifyTransforms(
    ItemToModify=((Item=(Type=Bone,Name="dyn_shorts_back_01_r"),Transform=(Rotation=(X=0,Y=0,Z=0,W=1),Translation=(X=2,Y=0,Z=0),Scale3D=(X=1,Y=1,Z=1))))
    Weight=@AlphaInterp.Result
   )
   > &RigUnit_ModifyTransforms(
    ItemToModify=((Item=(Type=Bone,Name="dyn_shorts_front_03_r"),Transform=(Rotation=(X=0,Y=0,Z=0,W=1),Translation=(X=10,Y=0,Z=0),Scale3D=(X=1,Y=1,Z=1))))
    Weight=@AlphaInterp.Result
   )
   > &RigUnit_ModifyTransforms(
    ItemToModify=((Item=(Type=Bone,Name="dyn_shorts_inner_03_r"),Transform=(Rotation=(X=0,Y=0,Z=0,W=1),Translation=(X=6,Y=0,Z=0),Scale3D=(X=1,Y=1,Z=1))))
    Weight=@AlphaInterp.Result
   )
   > &RigUnit_ModifyTransforms(
    ItemToModify=((Item=(Type=Bone,Name="dyn_shorts_inner_02_r"),Transform=(Rotation=(X=0,Y=0,Z=0,W=1),Translation=(X=6,Y=0,Z=0),Scale3D=(X=1,Y=1,Z=1))))
    Weight=@AlphaInterp.Result
   )
   > &RigUnit_ModifyTransforms(
    ItemToModify=((Item=(Type=Bone,Name="dyn_shorts_outer_03_r"),Transform=(Rotation=(X=0,Y=0,Z=0,W=1),Translation=(X=6,Y=0,Z=0),Scale3D=(X=1,Y=1,Z=1))))
    Weight=@AlphaInterp.Result
   )
   > &RigUnit_ModifyTransforms(
    ItemToModify=((Item=(Type=Bone,Name="dyn_shorts_outer_01_l"),Transform=(Rotation=(X=0,Y=0,Z=0,W=1),Translation=(X=0,Y=0,Z=0),Scale3D=(X=1,Y=1,Z=1))))
    Weight=1
   )
   n4 = RigVMDispatchNode(Array=RigVMVariableNode() {variable=BonesToMove_L}.Value.Keys, ExecuteContext=@SequenceAggregate.A) {template="DISPATCH_RigVMDispatch_ArrayIterator(in Array,out Element,out Index,out Count,out Ratio)"}
   > &RigUnit_ModifyTransforms(
    ItemToModify=((Item=(Type=Bone),Transform=(Rotation=(X=0,Y=0,Z=0,W=1),Translation=(X=-3,Y=0,Z=0),Scale3D=(X=1,Y=1,Z=1))))
    Weight=1
    ItemToModify.0.Item=@n4.Element
   )
   DrawAxis = RigVMFunctionReferenceNode(Active=true, ExecuteContext=(), Item=(Type=Bone,Name="calf_l"), Scale=15, Thickness=2)
   &RigUnit_ModifyTransforms(
    ExecutePin=()
    ItemToModify=((Item=(Type=Bone,Name="dyn_shorts_front_03_l"),Transform=(Rotation=(X=0,Y=0,Z=0,W=1),Translation=(X=0,Y=0,Z=0),Scale3D=(X=1,Y=1,Z=1))))
    Weight=@n2.Result
   )
  ]}
]
hierarchy [
 root {loc=<0 0 0>, rot=<0 0 0 1>}
 pelvis {parent=root, loc=<0 -0.7071 93.1191>, rot=<0.0127 0.707 -0.0127 -0.707>}
 spine_01 {parent=pelvis, loc=<6.7494 -0.6388 0>, rot=<0 0 0.034 -0.9994>}
 spine_02 {parent=spine_01, loc=<7.2518 -0.035 0>, rot=<0 0 0.0238 -0.9997>}
 spine_03 {parent=spine_02, loc=<6.0104 0 0>, rot=<0 0 0.0149 -0.9999>}
 spine_04 {parent=spine_03, loc=<8.1024 0 0>, rot=<0 0 -0.0132 -0.9999>}
 spine_05 {parent=spine_04, loc=<10.0146 0 0>, rot=<0 0 -0.0015 -1>}
 attach_cape {parent=spine_05, loc=<9.017 -11.714 0>, rot=<0.0616 -0.7044 -0.0616 -0.7044>}
 neck_01 {parent=spine_05, loc=<11.3138 0.1529 0>, rot=<0 0 -0.2003 -0.9797>}
 neck_02 {parent=neck_01, loc=<4.5627 0 0>, rot=<0 0 0.0504 -0.9987>}
 head {parent=neck_02, loc=<5.3914 0 0>, rot=<0 0 0.115 -0.9934>}
 attach_backpack {parent=spine_05, loc=<-0.123 -13.473 0>, rot=<-0.0093 -0.707 0.0093 -0.707>}
 clavicle_l {parent=spine_05, loc=<4.3575 -0.1466 -2.3223>, rot=<0.0465 -0.7289 -0.009 -0.6829>}
 upperarm_l {parent=clavicle_l, loc=<15.5111 0 0>, rot=<0.0226 -0.3234 -0.0088 -0.9459>}
 lowerarm_l {parent=upperarm_l, loc=<26.6514 0 0>, rot=<0.0391 -0.0099 -0.2453 -0.9686>}
 hand_l {parent=lowerarm_l, loc=<22.3491 0 0>, rot=<-0.7081 -0.0507 0.0975 -0.6975>}
 thumb_01_l {parent=hand_l, loc=<1.3118 -0.3191 -2.459>, rot=<0.6546 -0.4042 0.0664 -0.6354>}
 thumb_02_l {parent=thumb_01_l, loc=<5.9336 -0.0448 0.0401>, rot=<-0.0037 0.0213 0.298 -0.9543>}
 thumb_03_l {parent=thumb_02_l, loc=<2.2223 0.3682 -0.0907>, rot=<0.0118 -0.013 0.0066 -0.9998>}
 index_metacarpal_l {parent=hand_l, loc=<1.6535 0.0183 -2.96>, rot=<0.0158 -0.0124 -0.0173 -0.9996>}
 index_01_l {parent=index_metacarpal_l, loc=<8.3008 0.0357 -0.0524>, rot=<-0.0044 0.0485 0.0924 -0.9945>}
 index_02_l {parent=index_01_l, loc=<3.3587 -0.3004 0.0452>, rot=<-0.002 0.0027 0.3225 -0.9466>}
 index_03_l {parent=index_02_l, loc=<2.5578 -0.123 -0.0252>, rot=<-0.002 0.0067 0.2558 -0.9667>}
 middle_metacarpal_l {parent=hand_l, loc=<2.1856 -0.1725 -0.5984>, rot=<0.006 0.0168 -0.0329 -0.9993>}
 middle_01_l {parent=middle_metacarpal_l, loc=<7.3275 -0.2447 0.0129>, rot=<-0.0208 0.0042 0.0858 -0.9961>}
 middle_02_l {parent=middle_01_l, loc=<4.0123 0.1401 0.1175>, rot=<-0.0218 0.0258 0.3512 -0.9357>}
 middle_03_l {parent=middle_02_l, loc=<2.6077 0.0534 0.1373>, rot=<0.0082 0.0188 0.1708 -0.9851>}
 ring_metacarpal_l {parent=hand_l, loc=<2.2074 -0.3625 1.3638>, rot=<-0.0284 0.0722 -0.0458 -0.9959>}
 ring_01_l {parent=ring_metacarpal_l, loc=<7.0295 -0.6197 -0.1446>, rot=<-0.0489 -0.0252 0.1892 -0.9804>}
 ring_02_l {parent=ring_01_l, loc=<3.6174 0.4647 -0.1027>, rot=<0 0 0.2806 -0.9598>}
 ring_03_l {parent=ring_02_l, loc=<2.4152 0.114 0.0199>, rot=<0 0 0.1189 -0.9929>}
 pinky_metacarpal_l {parent=hand_l, loc=<1.7616 -0.9753 3.3814>, rot=<-0.0397 0.0814 -0.069 -0.9935>}
 pinky_01_l {parent=pinky_metacarpal_l, loc=<7.1709 -0.4831 -0.1094>, rot=<-0.0654 -0.0035 0.2242 -0.9723>}
 pinky_02_l {parent=pinky_01_l, loc=<2.8274 0.0893 -0.1043>, rot=<0.0037 0.0011 0.2854 -0.9584>}
 pinky_03_l {parent=pinky_02_l, loc=<2.2046 0.0307 -0.0078>, rot=<-0.0002 0.0098 0.1445 -0.9895>}
 lowerarm_twist_02_l {parent=lowerarm_l, loc=<9.6199 0 0>, rot=<0 0 0 -1>}
 lowerarm_twist_01_l {parent=lowerarm_l, loc=<17.8549 0 0>, rot=<0 0 0 -1>}
 upperarm_twist_01_l {parent=upperarm_l, loc=<0.0041 0 0>, rot=<0 0 0 1>}
 upperarm_twist_02_l {parent=upperarm_l, loc=<13.7941 0 0>, rot=<0 0 0 1>}
 clavicle_r {parent=spine_05, loc=<4.3575 -0.1466 2.3223>, rot=<0.7289 0.0465 -0.6829 0.009>}
 upperarm_r {parent=clavicle_r, loc=<-15.5111 0 0>, rot=<0.0226 -0.3234 -0.0088 -0.9459>}
 lowerarm_r {parent=upperarm_r, loc=<-26.6514 0 0>, rot=<0.0391 -0.0099 -0.2453 -0.9686>}
 hand_r {parent=lowerarm_r, loc=<-22.3491 0 0>, rot=<-0.7081 -0.0507 0.0975 -0.6975>}
 thumb_01_r {parent=hand_r, loc=<-1.3118 0.3191 2.459>, rot=<0.6546 -0.4042 0.0664 -0.6354>}
 thumb_02_r {parent=thumb_01_r, loc=<-5.9336 0.0448 -0.0401>, rot=<-0.0037 0.0213 0.298 -0.9543>}
 thumb_03_r {parent=thumb_02_r, loc=<-2.2223 -0.3682 0.0907>, rot=<0.0118 -0.013 0.0066 -0.9998>}
 index_metacarpal_r {parent=hand_r, loc=<-1.6535 -0.0183 2.96>, rot=<0.0158 -0.0124 -0.0173 -0.9996>}
 index_01_r {parent=index_metacarpal_r, loc=<-8.3008 -0.0357 0.0524>, rot=<-0.0044 0.0485 0.0924 -0.9945>}
 index_02_r {parent=index_01_r, loc=<-3.3587 0.3004 -0.0452>, rot=<-0.002 0.0027 0.3225 -0.9466>}
 index_03_r {parent=index_02_r, loc=<-2.5578 0.123 0.0252>, rot=<-0.002 0.0067 0.2558 -0.9667>}
 middle_metacarpal_r {parent=hand_r, loc=<-2.1856 0.1725 0.5984>, rot=<0.006 0.0168 -0.0329 -0.9993>}
 middle_01_r {parent=middle_metacarpal_r, loc=<-7.3275 0.2447 -0.0129>, rot=<-0.0208 0.0042 0.0858 -0.9961>}
 middle_02_r {parent=middle_01_r, loc=<-4.0123 -0.1401 -0.1175>, rot=<-0.0218 0.0258 0.3512 -0.9357>}
 middle_03_r {parent=middle_02_r, loc=<-2.6077 -0.0534 -0.1373>, rot=<0.0082 0.0188 0.1708 -0.9851>}
 ring_metacarpal_r {parent=hand_r, loc=<-2.2074 0.3625 -1.3638>, rot=<-0.0284 0.0722 -0.0458 -0.9959>}
 ring_01_r {parent=ring_metacarpal_r, loc=<-7.0295 0.6197 0.1446>, rot=<-0.0489 -0.0252 0.1892 -0.9804>}
 ring_02_r {parent=ring_01_r, loc=<-3.6174 -0.4647 0.1027>, rot=<0 0 0.2806 -0.9598>}
 ring_03_r {parent=ring_02_r, loc=<-2.4152 -0.114 -0.0199>, rot=<0 0 0.1189 -0.9929>}
 pinky_metacarpal_r {parent=hand_r, loc=<-1.7616 0.9753 -3.3814>, rot=<-0.0397 0.0814 -0.069 -0.9935>}
 pinky_01_r {parent=pinky_metacarpal_r, loc=<-7.1709 0.4831 0.1094>, rot=<-0.0654 -0.0035 0.2242 -0.9723>}
 pinky_02_r {parent=pinky_01_r, loc=<-2.8274 -0.0893 0.1043>, rot=<0.0037 0.0011 0.2854 -0.9584>}
 pinky_03_r {parent=pinky_02_r, loc=<-2.2046 -0.0307 0.0078>, rot=<-0.0002 0.0098 0.1445 -0.9895>}
 lowerarm_twist_02_r {parent=lowerarm_r, loc=<-9.6199 0 0>, rot=<0 0 0 -1>}
 lowerarm_twist_01_r {parent=lowerarm_r, loc=<-17.8549 0 0>, rot=<0 0 0 -1>}
 upperarm_twist_01_r {parent=upperarm_r, loc=<-0.0041 0 0>, rot=<0 0 0 -1>}
 upperarm_twist_02_r {parent=upperarm_r, loc=<-13.7941 0 0>, rot=<0 0 0 -1>}
 thigh_l {parent=pelvis, loc=<-2.5693 -0.0109 -7.7501>, rot=<0.0012 0.0572 0.0155 -0.9982>}
 calf_l {parent=thigh_l, loc=<-39.0484 0 0>, rot=<0.0006 -0.0161 -0.0391 -0.9991>}
 foot_l {parent=calf_l, loc=<-42.561 0.1077 0.4292>, rot=<0.0418 -0.0455 0.0343 -0.9975>}
 ball_l {parent=foot_l, loc=<-6.1599 13.8264 0.6806>, rot=<0.0001 -0.0001 -0.7063 -0.7079>}
 calf_twist_02_l {parent=calf_l, loc=<-18.8064 -0.0967 0.0397>, rot=<0.0237 0.0095 -0.0035 -0.9997>}
 calf_twist_01_l {parent=calf_l, loc=<-35.813 -0.0967 0.0397>, rot=<0.0237 0.0095 -0.0035 -0.9997>}
 dyn_loop_inner_01_l {parent=calf_l, loc=<-29.1733 4.2979 0.129>, rot=<0.3238 -0.429 0.7164 0.4449>}
 dyn_loop_inner_02_l {parent=dyn_loop_inner_01_l, loc=<1.7387 0 0>, rot=<-0.0099 -0.0174 0.0016 -0.9998>}
 dyn_loop_outer_01_l {parent=calf_l, loc=<-29.3276 4.0473 -0.9873>, rot=<-0.3524 0.5118 0.6956 0.3606>}
 dyn_loop_outer_02_l {parent=dyn_loop_outer_01_l, loc=<2.0804 0 0>, rot=<-0.0097 -0.0151 -0.0204 -0.9996>}
 dyn_lace_outer_01_l {parent=calf_l, loc=<-29.8402 4.2129 -0.5931>, rot=<-0.0371 0.5014 0.8613 0.0734>}
 dyn_lace_outer_02_l {parent=dyn_lace_outer_01_l, loc=<1.2718 0 0>, rot=<0.043 -0.0134 0.0507 -0.9977>}
 dyn_lace_outer_03_l {parent=dyn_lace_outer_02_l, loc=<1.7695 0 0>, rot=<0.024 -0.0033 0.0106 -0.9996>}
 dyn_lace_inner_01_l {parent=calf_l, loc=<-29.7597 4.2881 -0.1354>, rot=<0.0811 -0.3514 0.9236 0.1297>}
 dyn_lace_inner_02_l {parent=dyn_lace_inner_01_l, loc=<1.5056 0 0>, rot=<-0.0011 -0.0345 0.0128 -0.9993>}
 dyn_lace_inner_03_l {parent=dyn_lace_inner_02_l, loc=<2.0846 0 0>, rot=<0.0075 -0.0263 -0.0092 -0.9996>}
 thigh_twist_01_l {parent=thigh_l, loc=<-21.6303 0 0>, rot=<0.0013 0 0 -1>}
 dyn_shorts_inner_01_l {parent=thigh_l, loc=<-14.9018 -0.4703 8.7758>, rot=<0.0433 -0.7033 0.7092 0.0236>}
 dyn_shorts_inner_02_l {parent=dyn_shorts_inner_01_l, loc=<8.0526 0 0>, rot=<0 0 0 1>}
 dyn_shorts_inner_03_l {parent=dyn_shorts_inner_02_l, loc=<8.919 0 0>, rot=<0 0 0 1>}
 dyn_shorts_front_01_l {parent=thigh_l, loc=<-11.5973 10.0009 -1.2941>, rot=<0.013 0.049 0.9986 0.0148>}
 dyn_shorts_front_02_l {parent=dyn_shorts_front_01_l, loc=<9.9245 0 0>, rot=<0 0 0 1>}
 dyn_shorts_front_03_l {parent=dyn_shorts_front_02_l, loc=<11.2126 0 0>, rot=<0 0 0 1>}
 dyn_shorts_outer_01_l {parent=thigh_l, loc=<-8.7231 -1.8253 -8.817>, rot=<0.0091 -0.7662 -0.6426 0.0031>}
 dyn_shorts_outer_02_l {parent=dyn_shorts_outer_01_l, loc=<11.2813 0 0>, rot=<0 0 0 1>}
 dyn_shorts_outer_03_l {parent=dyn_shorts_outer_02_l, loc=<12.7651 0 0>, rot=<0 0 0 1>}
 dyn_shorts_back_01_l {parent=thigh_l, loc=<-9.7407 -12.601 -1.0893>, rot=<0.0203 -0.9993 0.0008 0.0319>}
 dyn_shorts_back_02_l {parent=dyn_shorts_back_01_l, loc=<11.0159 0 0>, rot=<0 0 0 1>}
 dyn_shorts_back_03_l {parent=dyn_shorts_back_02_l, loc=<11.9902 0 0>, rot=<0 0 0 1>}
 thigh_r {parent=pelvis, loc=<-2.5693 -0.0109 7.7501>, rot=<0.0572 -0.0012 0.9982 0.0155>}
 calf_r {parent=thigh_r, loc=<39.0484 0 0>, rot=<0.0006 -0.0161 -0.0391 -0.9991>}
 foot_r {parent=calf_r, loc=<42.561 -0.1077 -0.4292>, rot=<0.0418 -0.0455 0.0343 -0.9975>}
 ball_r {parent=foot_r, loc=<6.1599 -13.8264 -0.6806>, rot=<0.0001 -0.0001 -0.7063 -0.7079>}
 calf_twist_02_r {parent=calf_r, loc=<18.8064 0.0967 -0.0397>, rot=<0.0237 0.0095 -0.0035 -0.9997>}
 calf_twist_01_r {parent=calf_r, loc=<35.813 0.0967 -0.0397>, rot=<0.0237 0.0095 -0.0035 -0.9997>}
 dyn_loop_inner_01_r {parent=calf_r, loc=<29.2543 -4.121 0.0504>, rot=<-0.4069 -0.2934 0.4221 -0.7551>}
 dyn_loop_inner_02_r {parent=dyn_loop_inner_01_r, loc=<2.1344 0 0>, rot=<-0.0035 -0.0139 0.0513 -0.9986>}
 dyn_lace_inner_01_r {parent=calf_r, loc=<29.7966 -4.1411 0.2849>, rot=<-0.3178 -0.1011 0.0998 -0.9375>}
 dyn_lace_inner_02_r {parent=dyn_lace_inner_01_r, loc=<1.6918 0 0>, rot=<0.0025 -0.0351 0.0911 -0.9952>}
 dyn_lace_inner_03_r {parent=dyn_lace_inner_02_r, loc=<1.9602 0 0>, rot=<-0.0327 -0.0352 0.0322 -0.9983>}
 dyn_lace_outer_01_r {parent=calf_r, loc=<29.8894 -4.1392 0.7245>, rot=<0.4184 -0.0142 0.0746 -0.9051>}
 dyn_lace_outer_02_r {parent=dyn_lace_outer_01_r, loc=<1.669 0 0>, rot=<0.024 -0.0587 -0.0093 -0.9979>}
 dyn_lace_outer_03_r {parent=dyn_lace_outer_02_r, loc=<2.4629 0 0>, rot=<0.0253 -0.0542 -0.0476 -0.9971>}
 dyn_loop_outer_01_r {parent=calf_r, loc=<29.4866 -4.0603 1.0809>, rot=<0.5617 0.2516 0.3738 -0.6939>}
 dyn_loop_outer_02_r {parent=dyn_loop_outer_01_r, loc=<2.0216 0 0>, rot=<-0.0089 -0.0136 0.0323 -0.9993>}
 thigh_twist_01_r {parent=thigh_r, loc=<21.6303 0 0>, rot=<0.0013 0 0 -1>}
 dyn_shorts_inner_01_r {parent=thigh_r, loc=<14.9017 0.4703 -8.7758>, rot=<0.0433 -0.7033 0.7092 0.0236>}
 dyn_shorts_inner_02_r {parent=dyn_shorts_inner_01_r, loc=<-8.0526 0 0>, rot=<0 0 0 1>}
 dyn_shorts_inner_03_r {parent=dyn_shorts_inner_02_r, loc=<-8.9191 0 0>, rot=<0 0 0 1>}
 dyn_shorts_front_01_r {parent=thigh_r, loc=<11.6765 -10.0263 1.0591>, rot=<0.013 0.049 0.9986 0.0148>}
 dyn_shorts_front_02_r {parent=dyn_shorts_front_01_r, loc=<-9.9245 0 0>, rot=<0 0 0 1>}
 dyn_shorts_front_03_r {parent=dyn_shorts_front_02_r, loc=<-11.2126 0 0>, rot=<0 0 0 1>}
 dyn_shorts_outer_01_r {parent=thigh_r, loc=<8.7277 1.5487 8.8659>, rot=<0.0091 -0.7662 -0.6426 0.0031>}
 dyn_shorts_outer_02_r {parent=dyn_shorts_outer_01_r, loc=<-11.2812 0 0>, rot=<0 0 0 -1>}
 dyn_shorts_outer_03_r {parent=dyn_shorts_outer_02_r, loc=<-12.7652 0 0>, rot=<0 0 0 -1>}
 dyn_shorts_back_01_r {parent=thigh_r, loc=<9.784 12.603 1.7658>, rot=<0.0203 -0.9993 0.0008 0.0319>}
 dyn_shorts_back_02_r {parent=dyn_shorts_back_01_r, loc=<-11.0159 0 0>, rot=<0 0 0 -1>}
 dyn_shorts_back_03_r {parent=dyn_shorts_back_02_r, loc=<-11.9902 0 0>, rot=<0 0 0 -1>}
]
variables [
 BonesToMove_L {
  type=FRigElementKeyCollection
  default=(Keys=((Type=Bone,Name="dyn_shorts_back_01_l"),(Type=Bone,Name="dyn_shorts_outer_01_l"),(Type=Bone,Name="dyn_shorts_inner_01_l"),(Type=Bone,Name="dyn_shorts_front_01_l")))
 }
 BonesToMove_R {
  type=FRigElementKeyCollection
  default=(Keys=((Type=Bone,Name="dyn_shorts_back_01_r"),(Type=Bone,Name="dyn_shorts_outer_01_r"),(Type=Bone,Name="dyn_shorts_inner_01_r"),(Type=Bone,Name="dyn_shorts_front_01_r")))
 }
 calf_l_Alpha {type=double, default=0}
 calf_r_Alpha {type=double, default=0}
 isCrouching {type=bool, default=False}
 isJumping {type=bool, default=False}
]
