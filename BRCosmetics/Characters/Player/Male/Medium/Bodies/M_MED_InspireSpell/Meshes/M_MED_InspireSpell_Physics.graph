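# Physics asset for the M_MED_InspireSpell male medium body, covering secondary motion of the left shorts leg; it has no right-side bodies.
# Capsules on the left thigh and calf accompany four three-bone dyn_shorts chains: front, back, inner and outer.
# Each chain hangs from the thigh through constraints with very tight swing and twist limits, varied per chain, while the calf-to-thigh joint is angularly free.
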PhysicsAsset /BRCosmetics/Characters/Player/Male/Medium/Bodies/M_MED_InspireSpell/Meshes/M_MED_InspireSpell_Physics
bodies {
 calf_l [{shapeType=Capsule, length=20.7281, radius=6.6913, shapeName=calf_l_capsule, center=<-7.0936 -2.7098 0>, rotation=<rot -90 0 0>}]
 dyn_shorts_back_01_l [{shapeType=Capsule, length=13.9767, radius=1.3474, shapeName=dyn_shorts_outer_01_l_capsule, center=<5.2248 0 0>, rotation=<rot 0 180 180>}]
 dyn_shorts_back_02_l [{shapeType=Capsule, length=15.2545, radius=1.9036, shapeName=dyn_shorts_outer_02_l_capsule, center=<6.2616 0 0>}]
 dyn_shorts_back_03_l [{shapeType=Capsule, length=16.1966, radius=2.3617, shapeName=dyn_shorts_outer_03_l_capsule, center=<0.5911 0 0>}]
 dyn_shorts_front_01_l [{shapeType=Capsule, length=13.9767, radius=1.3474, shapeName=dyn_shorts_outer_01_l_capsule, center=<5.2248 0 0>, rotation=<rot 0 180 180>}]
 dyn_shorts_front_02_l [{shapeType=Capsule, length=15.2545, radius=1.9036, shapeName=dyn_shorts_outer_02_l_capsule, center=<6.2616 0 0>}]
 dyn_shorts_front_03_l [{shapeType=Capsule, length=16.1966, radius=2.0278, shapeName=dyn_shorts_outer_03_l_capsule, center=<0.5911 0 0>}]
 dyn_shorts_inner_01_l [{shapeType=Capsule, length=13.9767, radius=1.3474, shapeName=dyn_shorts_outer_01_l_capsule, center=<5.2248 0 0>, rotation=<rot 0 180 180>}]
 dyn_shorts_inner_02_l [{shapeType=Capsule, length=15.2545, radius=1.6276, shapeName=dyn_shorts_outer_02_l_capsule, center=<3.0408 0 0>}]
 dyn_shorts_inner_03_l [{shapeType=Capsule, length=16.1966, radius=1.8898, shapeName=dyn_shorts_outer_03_l_capsule, center=<0.5911 0 0>}]
 dyn_shorts_outer_01_l [{
   shapeType=Capsule
   length=13.9767
   radius=1.3474
   shapeName=dyn_shorts_outer_01_l_capsule
   center=<5.2248 -0.3816 0>
   rotation=<rot 0 180 180>
  }]
 dyn_shorts_outer_02_l [{shapeType=Capsule, length=15.2545, radius=1.5888, shapeName=dyn_shorts_outer_02_l_capsule, center=<6.2616 -0.5562 0>}]
 dyn_shorts_outer_03_l [{shapeType=Capsule, length=16.1966, radius=2.0369, shapeName=dyn_shorts_outer_03_l_capsule, center=<0.5911 -0.8983 0>}]
 thigh_l [{shapeType=Capsule, length=27.5771, radius=7.994, shapeName=thigh_l_capsule, center=<-19.4827 -0.226 0.0871>, rotation=<rot -90 180 180>}]
}
constraints [
 {bone1Name=dyn_shorts_back_01_l, bone2Name=thigh_l, swing1LimitDegrees=3, swing2LimitDegrees=3, twistLimitDegrees=2}
 {bone1Name=dyn_shorts_back_02_l, bone2Name=dyn_shorts_back_01_l, swing1LimitDegrees=3, swing2LimitDegrees=3, twistLimitDegrees=2}
 {bone1Name=dyn_shorts_back_03_l, bone2Name=dyn_shorts_back_02_l, swing1LimitDegrees=3, swing2LimitDegrees=3, twistLimitDegrees=2}
 {bone1Name=dyn_shorts_front_01_l, bone2Name=thigh_l, swing1LimitDegrees=4, swing2LimitDegrees=4, twistLimitDegrees=2}
 {bone1Name=dyn_shorts_front_02_l, bone2Name=dyn_shorts_front_01_l, swing1LimitDegrees=4, swing2LimitDegrees=4, twistLimitDegrees=2}
 {bone1Name=dyn_shorts_front_03_l, bone2Name=dyn_shorts_front_02_l, swing1LimitDegrees=4, swing2LimitDegrees=4, twistLimitDegrees=2}
 {bone1Name=dyn_shorts_inner_01_l, bone2Name=thigh_l, swing1LimitDegrees=1, swing2LimitDegrees=1, twistLimitDegrees=2}
 {bone1Name=dyn_shorts_inner_02_l, bone2Name=dyn_shorts_inner_01_l, swing1LimitDegrees=1, swing2LimitDegrees=1, twistLimitDegrees=2}
 {bone1Name=dyn_shorts_inner_03_l, bone2Name=dyn_shorts_inner_02_l, swing1LimitDegrees=2, swing2LimitDegrees=2, twistLimitDegrees=2}
 {bone1Name=dyn_shorts_outer_01_l, bone2Name=thigh_l, swing1LimitDegrees=2, swing2LimitDegrees=2, twistLimitDegrees=2}
 {bone1Name=dyn_shorts_outer_02_l, bone2Name=dyn_shorts_outer_01_l, swing1LimitDegrees=2, swing2LimitDegrees=2, twistLimitDegrees=2}
 {bone1Name=dyn_shorts_outer_03_l, bone2Name=dyn_shorts_outer_02_l, swing1LimitDegrees=4, swing2LimitDegrees=4, twistLimitDegrees=2}
 {
  bone1Name=calf_l
  bone2Name=thigh_l
  swing1LimitDegrees=45
  swing1Motion=Free
  swing2LimitDegrees=45
  swing2Motion=Free
  twistLimitDegrees=45
  twistMotion=Free
 }
]
