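# Post-process-style AnimBP for the M_MED_InspireSpell body cosmetic, adding secondary motion to the copied parent-mesh pose.
# Four thigh/calf-relative RigidBody sims, weakened while skydiving, feed two Control Rigs to the output.
# Each update also drives material-based WPO cloth simulation from both thighs, with wind during skydive floating.
# An additive shorts-bone offset chain keyed to jumping and crouching exists but is not connected to the output pose.
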AnimBlueprint /BRCosmetics/Characters/Player/Male/Medium/Bodies/M_MED_InspireSpell/Meshes/M_MED_InspireSpell_AnimBP
target_skeleton=&/M_MED_InspireSpell_Skeleton
graphs [
 AnimGraph {nodes [
   n1 = Get isJumping() -> (isJumping:bool)
   n2 = "Transform (Modify) Bone - Bone: dyn_shorts_back_03_r"!(ComponentPose, bAlphaBoolEnabled:bool=@n1, Translation="2,8,0", Rotation="0,-10,0", Scale="1,1,1") -> (Pose) {
    alphaInputType=Bool
    rotationMode=BMM_Additive
    rotationSpace=BCS_BoneSpace
    translationMode=BMM_Additive
    translationSpace=BCS_BoneSpace
    alphaBoolBlend {blendInTime=0.5, blendOption=Linear, blendOutTime=0.5}
    boneToModify {boneName=dyn_shorts_back_03_r}
    rotation {yaw=-10}
    translation=<2 8>
   }
   n3 = "Transform (Modify) Bone - Bone: dyn_shorts_back_02_r"!(ComponentPose=@n2, bAlphaBoolEnabled:bool=@n1, Translation="1,0,0", Rotation="0,-10,0", Scale="1,1,1") -> (Pose) {
    alphaInputType=Bool
    rotationMode=BMM_Additive
    rotationSpace=BCS_BoneSpace
    translationMode=BMM_Additive
    translationSpace=BCS_BoneSpace
    alphaBoolBlend {blendInTime=0.5, blendOption=Linear, blendOutTime=0.5}
    boneToModify {boneName=dyn_shorts_back_02_r}
    rotation {yaw=-10}
    translation {x=1}
   }
   n4 = "Transform (Modify) Bone - Bone: dyn_shorts_back_01_r"!(ComponentPose=@n3, bAlphaBoolEnabled:bool=@n1, Translation="1,0,0", Rotation="0,-10,0", Scale="1,1,1") -> (Pose) {
    alphaInputType=Bool
    rotationMode=BMM_Additive
    rotationSpace=BCS_BoneSpace
    translationMode=BMM_Additive
    translationSpace=BCS_BoneSpace
    alphaBoolBlend {blendInTime=0.5, blendOption=Linear, blendOutTime=0.5}
    boneToModify {boneName=dyn_shorts_back_01_r}
    rotation {yaw=-10}
    translation {x=1}
   }
   n5 = "Transform (Modify) Bone - Bone: dyn_shorts_inner_03_r"!(ComponentPose=@n4, bAlphaBoolEnabled:bool=@n1, Translation="12,0,0", Rotation="0,-10,0", Scale="1,1,1") -> (Pose) {
    alphaInputType=Bool
    rotationMode=BMM_Additive
    rotationSpace=BCS_BoneSpace
    translationMode=BMM_Additive
    translationSpace=BCS_BoneSpace
    alphaBoolBlend {blendInTime=0.5, blendOption=Linear, blendOutTime=0.5}
    boneToModify {boneName=dyn_shorts_inner_03_r}
    rotation {yaw=-10}
    translation {x=12}
   }
   n6 = "Transform (Modify) Bone - Bone: dyn_shorts_front_03_r"!(ComponentPose=@n5, bAlphaBoolEnabled:bool=@n1, Translation="20,0,0", Rotation="0,-10,0", Scale="1,1,1") -> (Pose) {
    alphaInputType=Bool
    rotationMode=BMM_Additive
    rotationSpace=BCS_BoneSpace
    translationMode=BMM_Additive
    translationSpace=BCS_BoneSpace
    alphaBoolBlend {blendInTime=0.5, blendOption=Linear, blendOutTime=0.5}
    boneToModify {boneName=dyn_shorts_front_03_r}
    rotation {yaw=-10}
    translation {x=20}
   }
   n7 = "Transform (Modify) Bone - Bone: dyn_shorts_outer_03_r"!(ComponentPose=@n6, bAlphaBoolEnabled:bool=@n1, Translation="12,0,0", Rotation="0,-10,0", Scale="1,1,1") -> (Pose) {
    alphaInputType=Bool
    rotationMode=BMM_Additive
    rotationSpace=BCS_BoneSpace
    translationMode=BMM_Additive
    translationSpace=BCS_BoneSpace
    alphaBoolBlend {blendInTime=0.5, blendOption=Linear, blendOutTime=0.5}
    boneToModify {boneName=dyn_shorts_outer_03_r}
    rotation {yaw=-10}
    translation {x=12}
   }
   n8 = "Transform (Modify) Bone - Bone: dyn_shorts_back_03_l"!(ComponentPose=@n7, bAlphaBoolEnabled:bool=@n1, Translation="-2,-2,0", Rotation="0,10,0", Scale="1,1,1") -> (Pose) {
    alphaInputType=Bool
    rotationMode=BMM_Additive
    rotationSpace=BCS_BoneSpace
    translationMode=BMM_Additive
    translationSpace=BCS_BoneSpace
    alphaBoolBlend {blendInTime=0.5, blendOption=Linear, blendOutTime=0.5}
    boneToModify {boneName=dyn_shorts_back_03_l}
    rotation {yaw=10}
    translation=<-2 -2>
   }
   n9 = "Transform (Modify) Bone - Bone: dyn_shorts_back_02_l"!(ComponentPose=@n8, bAlphaBoolEnabled:bool=@n1, Translation="-2,-2,0", Rotation="0,10,0", Scale="1,1,1") -> (Pose) {
    alphaInputType=Bool
    rotationMode=BMM_Additive
    rotationSpace=BCS_BoneSpace
    translationMode=BMM_Additive
    translationSpace=BCS_BoneSpace
    alphaBoolBlend {blendInTime=0.5, blendOption=Linear, blendOutTime=0.5}
    boneToModify {boneName=dyn_shorts_back_02_l}
    rotation {yaw=10}
    translation=<-2 -2>
   }
   n10 = "Transform (Modify) Bone - Bone: dyn_shorts_inner_03_l"!(ComponentPose=@n9, bAlphaBoolEnabled:bool=@n1, Translation="-10,0,0", Rotation="0,-10,0", Scale="1,1,1") -> (Pose) {
    alphaInputType=Bool
    rotationMode=BMM_Additive
    rotationSpace=BCS_BoneSpace
    translationMode=BMM_Additive
    translationSpace=BCS_BoneSpace
    alphaBoolBlend {blendInTime=0.5, blendOption=Linear, blendOutTime=0.5}
    boneToModify {boneName=dyn_shorts_inner_03_l}
    rotation {yaw=-10}
    translation {x=-10}
   }
   n11 = "Transform (Modify) Bone - Bone: dyn_shorts_front_03_l"!(ComponentPose=@n10, bAlphaBoolEnabled:bool=@n1, Translation="-8,0,0", Rotation="0,-10,0", Scale="1,1,1") -> (Pose) {
    alphaInputType=Bool
    rotationMode=BMM_Additive
    rotationSpace=BCS_BoneSpace
    translationMode=BMM_Additive
    translationSpace=BCS_BoneSpace
    alphaBoolBlend {blendInTime=0.5, blendOption=Linear, blendOutTime=0.5}
    boneToModify {boneName=dyn_shorts_front_03_l}
    rotation {yaw=-10}
    translation {x=-8}
   }
   n12 = "Transform (Modify) Bone - Bone: dyn_shorts_outer_03_l"!(ComponentPose=@n11, bAlphaBoolEnabled:bool=@n1, Translation="-12,0,0", Rotation="0,-10,0", Scale="1,1,1") -> (Pose) {
    alphaInputType=Bool
    rotationMode=BMM_Additive
    rotationSpace=BCS_BoneSpace
    translationMode=BMM_Additive
    translationSpace=BCS_BoneSpace
    alphaBoolBlend {blendInTime=0.5, blendOption=Linear, blendOutTime=0.5}
    boneToModify {boneName=dyn_shorts_outer_03_l}
    rotation {yaw=-10}
    translation {x=-12}
   }
   n13 = Get isCrouching() -> (isCrouching:bool)
   n14 = "Transform (Modify) Bone - Bone: dyn_shorts_back_03_r"(ComponentPose=@n12, bAlphaBoolEnabled=@n13, Translation="4,0,0", Rotation="0,0,0", Scale="1,1,1") {
    alphaInputType=Bool
    rotationMode=BMM_Additive
    rotationSpace=BCS_BoneSpace
    translationMode=BMM_Additive
    translationSpace=BCS_BoneSpace
    alphaBoolBlend {blendInTime=0.1, blendOption=Linear, blendOutTime=0.1}
    boneToModify {boneName=dyn_shorts_back_03_r}
    translation {x=4}
   }
   n15 = "Transform (Modify) Bone - Bone: dyn_shorts_back_02_r"(ComponentPose=@n14, bAlphaBoolEnabled=@n13, Translation="6,-2,0", Rotation="0,0,0", Scale="1,1,1") {
    alphaInputType=Bool
    rotationMode=BMM_Additive
    rotationSpace=BCS_BoneSpace
    translationMode=BMM_Additive
    translationSpace=BCS_BoneSpace
    alphaBoolBlend {blendInTime=0.1, blendOption=Linear, blendOutTime=0.1}
    boneToModify {boneName=dyn_shorts_back_02_r}
    translation=<6 -2>
   }
   n16 = "Transform (Modify) Bone - Bone: dyn_shorts_back_01_r"(ComponentPose=@n15, bAlphaBoolEnabled=@n13, Translation="4,0,0", Rotation="0,0,0", Scale="1,1,1") {
    alphaInputType=Bool
    rotationMode=BMM_Additive
    rotationSpace=BCS_BoneSpace
    translationMode=BMM_Additive
    translationSpace=BCS_BoneSpace
    alphaBoolBlend {blendInTime=0.1, blendOption=Linear, blendOutTime=0.1}
    boneToModify {boneName=dyn_shorts_back_01_r}
    translation {x=4}
   }
   n17 = "Transform (Modify) Bone - Bone: dyn_shorts_front_03_r"(ComponentPose=@n16, bAlphaBoolEnabled=@n13, Translation="8,0,0", Rotation="0,0,0", Scale="1,1,1") {
    alphaInputType=Bool
    rotationMode=BMM_Additive
    rotationSpace=BCS_BoneSpace
    translationMode=BMM_Additive
    translationSpace=BCS_BoneSpace
    alphaBoolBlend {blendInTime=0.1, blendOption=Linear, blendOutTime=0.1}
    boneToModify {boneName=dyn_shorts_front_03_r}
    translation {x=8}
   }
   n18 = "Transform (Modify) Bone - Bone: dyn_shorts_inner_03_r"(ComponentPose=@n17, bAlphaBoolEnabled=@n13, Translation="16,0,0", Rotation="0,0,0", Scale="1,1,1") {
    alphaInputType=Bool
    rotationMode=BMM_Additive
    rotationSpace=BCS_BoneSpace
    translationMode=BMM_Additive
    translationSpace=BCS_BoneSpace
    alphaBoolBlend {blendInTime=0.1, blendOption=Linear, blendOutTime=0.1}
    boneToModify {boneName=dyn_shorts_inner_03_r}
    translation {x=16}
   }
   n19 = "Transform (Modify) Bone - Bone: dyn_shorts_outer_03_r"(ComponentPose=@n18, bAlphaBoolEnabled=@n13, Translation="16,0,0", Rotation="0,0,0", Scale="1,1,1") {
    alphaInputType=Bool
    rotationMode=BMM_Additive
    rotationSpace=BCS_BoneSpace
    translationMode=BMM_Additive
    translationSpace=BCS_BoneSpace
    alphaBoolBlend {blendInTime=0.1, blendOption=Linear, blendOutTime=0.1}
    boneToModify {boneName=dyn_shorts_outer_03_r}
    translation {x=16}
   }
   n20 = "Transform (Modify) Bone - Bone: dyn_shorts_back_03_l"(ComponentPose=@n19, bAlphaBoolEnabled=@n13, Translation="-10,0,0", Rotation="0,-10,0", Scale="1,1,1") {
    alphaInputType=Bool
    rotationMode=BMM_Additive
    rotationSpace=BCS_BoneSpace
    translationMode=BMM_Additive
    translationSpace=BCS_BoneSpace
    alphaBoolBlend {blendInTime=0.1, blendOption=Linear, blendOutTime=0.1}
    boneToModify {boneName=dyn_shorts_back_03_l}
    rotation {yaw=-10}
    translation {x=-10}
   }
   n21 = "Transform (Modify) Bone - Bone: dyn_shorts_back_02_l"(ComponentPose=@n20, bAlphaBoolEnabled=@n13, Translation="-8,0,0", Rotation="0,0,0", Scale="1,1,1") {
    alphaInputType=Bool
    rotationMode=BMM_Additive
    rotationSpace=BCS_BoneSpace
    translationMode=BMM_Additive
    translationSpace=BCS_BoneSpace
    alphaBoolBlend {blendInTime=0.1, blendOption=Linear, blendOutTime=0.1}
    boneToModify {boneName=dyn_shorts_back_02_l}
    translation {x=-8}
   }
   n22 = "Transform (Modify) Bone - Bone: dyn_shorts_front_03_l"(ComponentPose=@n21, bAlphaBoolEnabled=@n13, Translation="0,0,0", Rotation="0,0,0", Scale="1,1,1") {
    alphaInputType=Bool
    rotationMode=BMM_Additive
    rotationSpace=BCS_BoneSpace
    translationMode=BMM_Additive
    translationSpace=BCS_BoneSpace
    alphaBoolBlend {blendInTime=0.1, blendOption=Linear, blendOutTime=0.1}
    boneToModify {boneName=dyn_shorts_front_03_l}
   }
   n23 = "Transform (Modify) Bone - Bone: dyn_shorts_inner_03_l"(ComponentPose=@n22, bAlphaBoolEnabled=@n13, Translation="-16,0,0", Rotation="0,0,0", Scale="1,1,1") {
    alphaInputType=Bool
    rotationMode=BMM_Additive
    rotationSpace=BCS_BoneSpace
    translationMode=BMM_Additive
    translationSpace=BCS_BoneSpace
    alphaBoolBlend {blendInTime=0.1, blendOption=Linear, blendOutTime=0.1}
    boneToModify {boneName=dyn_shorts_inner_03_l}
    translation {x=-16}
   }
   "Transform (Modify) Bone - Bone: dyn_shorts_outer_03_l"(ComponentPose=@n23, bAlphaBoolEnabled=@n13, Translation="-10,0,0", Rotation="0,0,0", Scale="1,1,1") {
    alphaInputType=Bool
    rotationMode=BMM_Additive
    rotationSpace=BCS_BoneSpace
    translationMode=BMM_Additive
    translationSpace=BCS_BoneSpace
    alphaBoolBlend {blendInTime=0.5, blendOption=Linear, blendOutTime=0.5}
    boneToModify {boneName=dyn_shorts_outer_03_l}
    translation {x=-10}
   }
   n24 = Get skydiveAlpha() -> (skydiveAlpha:double)
   n25 = RigidBody(ComponentPose=Local To Component(LocalPose=Copy Pose From Mesh() {bUseAttachedParent=true}), Alpha=@n24) {
    bOverrideWorldGravity=true
    overridePhysicsAsset=&/M_MED_InspireSpell_Physics
    simulationSpace=BaseBoneSpace
    alphaBoolBlend {blendOption=Linear}
    alphaScaleBiasClamp {bInterpResult=true}
    baseBoneRef {boneName=thigh_l}
    componentAppliedLinearAccClamp=<3000 3000 3000>
    componentLinearAccScale=<1 1 1>
    componentLinearVelScale=<0.3 0.3 0.3>
    overrideWorldGravity {z=-980}
    simSpaceSettings {maxAngularAcceleration=3000, maxAngularVelocity=3000, maxLinearAcceleration=3000, maxLinearVelocity=3000}
   }
   n26 = RigidBody(ComponentPose=@n25, Alpha=@n24) {
    bOverrideWorldGravity=true
    overridePhysicsAsset=&/M_MED_InspireSpell_CalfL_Physics
    simulationSpace=BaseBoneSpace
    alphaBoolBlend {blendOption=Linear}
    alphaScaleBiasClamp {bInterpResult=true}
    baseBoneRef {boneName=calf_l}
    componentAppliedLinearAccClamp=<3000 3000 3000>
    componentLinearAccScale=<1 1 1>
    componentLinearVelScale=<0.3 0.3 0.3>
    overrideWorldGravity {z=-980}
    simSpaceSettings {maxAngularAcceleration=3000, maxAngularVelocity=3000, maxLinearAcceleration=3000, maxLinearVelocity=3000}
   }
   n27 = RigidBody(ComponentPose=@n26, Alpha=@n24) {
    bOverrideWorldGravity=true
    overridePhysicsAsset=&/M_MED_InspireSpell_ThighR_Physics
    simulationSpace=BaseBoneSpace
    alphaBoolBlend {blendOption=Linear}
    alphaScaleBiasClamp {bInterpResult=true}
    baseBoneRef {boneName=thigh_r}
    componentAppliedLinearAccClamp=<3000 3000 3000>
    componentLinearAccScale=<1 1 1>
    componentLinearVelScale=<0.3 0.3 0.3>
    overrideWorldGravity {z=-980}
    simSpaceSettings {maxAngularAcceleration=3000, maxAngularVelocity=3000, maxLinearAcceleration=3000, maxLinearVelocity=3000}
   }
   n28 = RigidBody(ComponentPose=@n27, Alpha=@n24) {
    bOverrideWorldGravity=true
    overridePhysicsAsset=&/M_MED_InspireSpell_CalfR_Physics
    simulationSpace=BaseBoneSpace
    alphaBoolBlend {blendOption=Linear}
    alphaScaleBiasClamp {bInterpResult=true}
    baseBoneRef {boneName=calf_r}
    componentAppliedLinearAccClamp=<3000 3000 3000>
    componentLinearAccScale=<1 1 1>
    componentLinearVelScale=<0.3 0.3 0.3>
    overrideWorldGravity {z=-980}
    simSpaceSettings {maxAngularAcceleration=3000, maxAngularVelocity=3000, maxLinearAcceleration=3000, maxLinearVelocity=3000}
   }
   n29 = Control Rig(Source=Component To Local(ComponentPose=@n28)) {
    alphaBoolBlend {blendOption=Linear}
    controlRigAssetReference {blueprintRigClass=&/M_MED_InspireSpell_CtrlRig.M_MED_InspireSpell_CtrlRig_C}
   }
   n30 = Control Rig(Source=@n29) {
    alphaBoolBlend {blendOption=Linear}
    controlRigAssetReference {blueprintRigClass=&/M_MED_InspireSpell_Deform_ControlRig.M_MED_InspireSpell_Deform_ControlRig_C}
   }
   Output Pose(Result=@n30) {layerGroup=None}
  ]}
 EventGraph {nodes [
   n1 = Event BlueprintUpdateAnimation()
   > n2 = Sequence()
   > Is Valid(InputObject=Get playerPawn() -> (playerPawn:Fort Player Pawn Athena Object Reference))
   > Set WPO Globals Function(
    Material=Get matInstanceDynamic() -> (matInstanceDynamic:Material Instance Dynamic Object Reference)
    Overlay Material=Get overlayMID() -> (overlayMID:Material Instance Dynamic Object Reference)
    Skeletal Mesh=Get skelMesh() -> (skelMesh:Skeletal Mesh Component Object Reference)
    "Local Noise Transform  Location"="Get Local Noise Transform  Location"() -> ("Local Noise Transform  Location":Vector)
    Local Noise Offset=Get Local Noise Offset()
   )
   > Set WPO Regional Function!(
    execute
    Material:Material Instance Dynamic Object Reference=Get matInstanceDynamic()
    Overlay Material:Material Instance Dynamic Object Reference=Get overlayMID()
    Skeletal Mesh:Skeletal Mesh Component Object Reference=Get skelMesh()
    Velocity Transform:Name=thigh_l
    Velocity Interp Speed:double=40
    Collision Transform:Name=thigh_l
    Collision Transform Axis:EAxis Enum
    Invert Collision Transform Axis:bool=true
    Delta Time X:double=@n1.DeltaTimeX
    Previous Velocity Transform Location:"Vector (by ref)"=Get R1 Velocity Transform Location()
    Previous Velocity Transform Velocity:"Vector (by ref)"=Get R1 Previous Velocity Transform Velocity() -> (R1 Previous Velocity Transform Velocity:Vector)
    Region:int=1
   ) -> (then)
   > Set WPO Regional Function(
    Material=Get matInstanceDynamic()
    Overlay Material=Get overlayMID()
    Skeletal Mesh=Get skelMesh()
    Velocity Transform=thigh_r
    Velocity Interp Speed=40
    Collision Transform=thigh_r
    Invert Collision Transform Axis=false
    Delta Time X=@n1.DeltaTimeX
    Previous Velocity Transform Location=Get R2 Velocity Transform Location()
    Previous Velocity Transform Velocity=Get R2 Velocity Transform Velocity()
    Region=2
   )
   > n3 = Branch(Condition=Get bIsSkydiveFloating())
   > n4 = Set WPO Pho Wind Intensity(WPO Pho Wind Intensity=1)
   SetScalarParameterValue(execute=[@n5.then, @n4.then], self=Get matInstanceDynamic(), ParameterName=ClothSim_WindStrength, Value=Get WPO Pho Wind Intensity())
   n5 = Set WPO Pho Wind Intensity(execute=@n3.else, WPO Pho Wind Intensity=0)
   Set isCrouching(execute=@n2.then_1, isCrouching:bool=Get bIsCrouching()) -> (Output_Get:bool)
   > Set isJumping(isJumping:bool=Get bIsJumping()) -> (Output_Get:bool)
   > n6 = Branch(Condition=Get bIsSkyDiving())
   > Set skydiveAlpha(skydiveAlpha:double=0.2) -> (Output_Get:double)
   Set skydiveAlpha(execute=@n6.else, skydiveAlpha=1)
   Event BlueprintInitializeAnimation()
   > Set SkelMesh(SkelMesh=GetOwnerSkeletalMeshComponent())
   n7 = TryGetPawnOwner()
   > Is Valid(InputObject=@n7)
   > n8 = Cast To FortPlayerPawnAthena(Object=@n7)
   > Set playerPawn(playerPawn:Fort Player Pawn Athena Object Reference=@n8) -> (Output_Get:Fort Player Pawn Athena Object Reference)
   Event OnCharacterCustomizationLoaded()
   > n9 = Set skelMesh(skelMesh:Skeletal Mesh Component Object Reference=GetPartSkeletalMeshComponent()) -> (Output_Get:Skeletal Mesh Component Object Reference)
   > n10 = Get Dynamic Material Instance(SkeletalMeshComponent=@n9)
   > Set matInstanceDynamic(matInstanceDynamic:Material Instance Dynamic Object Reference=@n10) -> (Output_Get:Material Instance Dynamic Object Reference)
   > n11 = GetOrCreateDynamicOverlayMaterial(MeshComponent=@n9, MaterialName=None)
   > Set overlayMID(overlayMID:Material Instance Dynamic Object Reference=@n11) -> (Output_Get:Material Instance Dynamic Object Reference)
  ]}
]
variables [
 isCrouching {type=boolean}
 isJumping {type=boolean}
 Local Noise Offset {type=Vector}
 "Local Noise Transform  Location" {type=Vector}
 matInstanceDynamic {type=MaterialInstanceDynamic}
 overlayMID {type=MaterialInstanceDynamic}
 playerPawn {type=FortPlayerPawnAthena}
 R1 Previous Velocity Transform Velocity {type=Vector}
 R1 Velocity Transform Location {type=Vector}
 R2 Velocity Transform Location {type=Vector}
 R2 Velocity Transform Velocity {type=Vector}
 skelMesh {type=SkeletalMeshComponent}
 skydiveAlpha {type=number}
 WPO Pho Wind Intensity {type=number}
]
